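# Landscape material function that builds a procedural snow layer, outputting snow material, blend mask, grass mask and simplified mobile material.
# A world-height ramp limited by a landscape-space sphere mask drives the grass mask and a HeightLerp transition.
# HeightLerp's height comes from snow texture alpha weighted by the input normal's Z, which also thins the final mask on tilted detail.
# The mobile variant drops the normal map and shades albedo by vertex-normal slope.
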MaterialFunction /Clyde/Athena/Apollo/Environments/Landscape/Materials/Biome/UniqueBiomeMaterialFunctions/MF_ProceduralSnow
graph {nodes [
  AppendVector(
   A=ScalarParameter() {defaultValue=275, group=ProceduralSnow, parameterName=SnowRadMaskoffseta}
   B=ScalarParameter() {defaultValue=1675, group=ProceduralSnow, parameterName=SnowRadMaskoffsetb}
  )
  Constant() {r=4800}
  Constant() {r=5000}
  Constant() {r=-5000}
  Constant() {r=-5200}
  Constant() {r=275}
  Constant() {r=1675}
  n1 = ScalarParameter() {defaultValue=0.6, group=ProceduralSnow, parameterName=SnowAlbedo-Bright}
  n2 = LandscapeLayerCoords() {mappingScale=30}
  n3 = TextureSample(UVs=@n2) {samplerSource=SSM_Wrap_WorldGroupSettings, samplerType=SAMPLERTYPE_Normal, texture=/Clyde/ExtraDependencies/T_Snow01_N}
  n4 = Constant3Vector() {constant=<rgba 0.25 0.25 1 0>}
  n5 = TextureSample(UVs=Multiply(A=ScalarParameter() {defaultValue=3, parameterName=SnowRoughTile}, B=@n2)) {
   samplerSource=SSM_Wrap_WorldGroupSettings
   samplerType=SAMPLERTYPE_LinearColor
   texture=/Game/Athena/Environments/Landscape/Textures/T_Athena_Snow_01_M
  }
  FunctionOutput(None=SetMaterialAttributes!(MaterialAttributes, BaseColor=@n1, Roughness=@n5.R, Normal=Multiply(A=@n3.RGB, B=@n4)) {attributeSetTypes [#1, #2, #3]}) {outputName=Material}
  n6 = TextureSample(UVs=LandscapeLayerCoords() {mappingScale=50}) {
   samplerSource=SSM_Wrap_WorldGroupSettings
   samplerType=SAMPLERTYPE_LinearColor
   texture=&T_Athena_Snow_01_M
  }
  n7 = GetMaterialAttributes(None=FunctionInput() {inputName=MaterialIn, inputType=FunctionInput_MaterialAttributes}) {attributeGetTypes [#3]}
  n8 = Saturate(None=ComponentMask(None=@n7.Normal) {b=true})
  n9 = Power(Base=@n8, Exp=Constant() {r=3})
  n10 = Constant() {r=1.75}
  n11 = Multiply(A=@n6.A, B=Saturate(None=Multiply(A=@n9, B=@n10)))
  n12 = Constant() {r=0.5}
  n13 = Add(
   A=ComponentMask(None=WorldPosition().XYZ) {b=true}
   B=ScalarParameter() {defaultValue=-5500, group=ProceduralSnow, parameterName=SnowBlendHeightOffset}
  )
  n14 = ScalarParameter() {defaultValue=4000, group=ProceduralSnow, parameterName=SnowBlendDistance}
  n15 = Multiply(
   A=SphereMask(A=LandscapeLayerCoords(), B=Constant2Vector() {g=1675, r=275}, Radius=Constant() {r=700}, Hardness=Constant() {r=0.75}) {hardnessPercent=0}
   B=Saturate(None=Divide(A=@n13, B=@n14))
  )
  n16 = Add(A=Constant() {r=-0.75}, B=Multiply(A=Constant() {r=2}, B=@n8))
  FunctionOutput(None=Multiply(
    A=MaterialFunctionCall(Transition Phase=@n15, Height Texture=@n11, Contrast=@n12) {materialFunction=&HeightLerp}.Alpha
    B=Saturate(None=@n16)
   )) {outputName=MaterialMask, sortPriority=1}
  n17 = Constant() {r=1.5}
  FunctionOutput(None=Saturate(None=Multiply(A=@n15, B=@n17))) {outputName=GrassMask, sortPriority=2}
  n18 = Abs(None=ComponentMask(None=VertexNormalWS()) {b=true})
  n19 = Constant() {r=1.25}
  n20 = Saturate(None=Add(A=Multiply(A=@n18, B=@n19), B=Constant() {r=-0.5}))
  n21 = Max(A=@n20, B=Constant() {r=0.35})
  n22 = ScalarParameter() {defaultValue=0.6, group=ProceduralSnow, parameterName=SnowAlbedo-Mobile}
  FunctionOutput(None=SetMaterialAttributes!(MaterialAttributes, BaseColor=Multiply(A=@n21, B=@n22), Roughness=@n5.R) {attributeSetTypes [#1, #2]}) {outputName=MobileMaterial}
  Multiply(A=Constant() {r=0.5}, B=@n2)
  Saturate(None=LinearInterpolate(A=ScalarParameter() {defaultValue=0.7, group=ProceduralSnow, parameterName=SnowAlbedo-Dark}, B=@n1, Alpha=@n20))
  Transform(None=@n7.Normal)
 ]}
